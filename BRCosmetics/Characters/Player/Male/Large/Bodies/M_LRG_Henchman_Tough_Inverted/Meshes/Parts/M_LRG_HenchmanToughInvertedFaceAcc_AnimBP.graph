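# Animation Blueprint for the face accessory part of the large male Henchman Tough Inverted player character.
# It copies the pose from the attached parent mesh, then a Transform (Modify) Bone node replaces the head_alt rotation in bone space with HeadRotation, blended at partial alpha.
# HeadRotation is not declared in this Blueprint's variables, and the event graph holds only unconnected default nodes.
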AnimBlueprint /BRCosmetics/Characters/Player/Male/Large/Bodies/M_LRG_Henchman_Tough_Inverted/Meshes/Parts/M_LRG_HenchmanToughInvertedFaceAcc_AnimBP
target_skeleton=&/M_LRG_HenchmanToughInvertedFaceAcc_Skeleton
graphs [
 AnimGraph {nodes [
   n1 = "Transform (Modify) Bone - Bone: head_alt"(
    ComponentPose=Local To Component(LocalPose=Copy Pose From Mesh() {bUseAttachedParent=true})
    Alpha=0.3
    Translation="0,0,0"
    Rotation=Get HeadRotation() -> (HeadRotation:Rotator)
    Scale="1,1,1"
   ) {
    rotationMode=BMM_Replace
    rotationSpace=BCS_BoneSpace
    translationSpace=BCS_BoneSpace
    alphaBoolBlend {blendOption=Linear}
    alphaScaleBiasClamp {interpSpeedDecreasing=1, interpSpeedIncreasing=1}
    boneToModify {boneName=head_alt}
   }
   Output Pose(Result=Component To Local(ComponentPose=@n1)) {layerGroup=None}
  ]}
 EventGraph {nodes [
   TryGetPawnOwner()
   Event BlueprintUpdateAnimation()
  ]}
]
variables []
